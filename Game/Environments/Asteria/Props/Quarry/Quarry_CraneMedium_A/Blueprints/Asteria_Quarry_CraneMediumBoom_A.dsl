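; Quarry crane boom prop with a hanging claw on a rope, posed at construction from editable claw-open and rope-length values.
; The construction script mirrors rotation on two claw halves, offsets the claw set vertically, then stretches the downward-pointing rope mesh to span its sockets.
; The rope's stretch factor, derived from mesh bounds, also feeds custom primitive data for its material.
; On death, the looping wood creak audio fades out.
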(event EventOnDeathPlayEffects
  (params (Damage :type Float%20%28single-precision%29) (DamageTags :type Gameplay%20Tag%20Container%20Structure%20%28by%20ref%29) (Momentum :type Vector :default "0, 0, 0") (HitInfo :type Hit%20Result%20Structure%20%28by%20ref%29) (InstigatedBy :type Fort%20Pawn%20Object%20Reference) (DamageCauser :type Actor%20Object%20Reference) (EffectContext :type Gameplay%20Effect%20Context%20Handle%20Structure))
  (bind _wood_creak_mediumhigh_loop_cue (Variables|Default|GetWood_Creak_MediumHigh_Loop_Cue))
  (Utilities|IsValid _wood_creak_mediumhigh_loop_cue
    (:Is%20Valid
      (if (Utilities|IsValid _wood_creak_mediumhigh_loop_cue :!class /Script/Engine.KismetSystemLibrary)
        (Audio|Components|Audio|FadeOut _wood_creak_mediumhigh_loop_cue 0.2 0.0 "Linear" :!class /Script/Engine.AudioComponent)))))

(fn ConstructionScript
  (Utilities|FlowControl|Sequence
    (:then_0
      (Default|ValidateClawOpen))
    (:then_1
      (Default|ValidateClawTransform))
    (:then_2
      (Default|ValidateBoomEndRope))))

(fn ValidateBoomEndRope
  (bind (_returnvalue_origin _returnvalue_boxextent) (Math|BoxSphereBounds|BreakBoxSphereBounds (StaticMesh|GetBounds (Variables|/Script/Engine.StaticMeshComponent|GetStaticMesh (Variables|Default|GetBoomRope)))))
  (bind _returnvalue_boxextent_x (Math|Vector|BreakVector _returnvalue_boxextent))
  (LocalVariables|Setl_start (Transformation|GetSocketLocation (Variables|Debug|GetStaticMeshComponent) "RopeSocket_End"))
  (LocalVariables|Setl_end (Transformation|GetSocketLocation (Variables|Default|GetSM_Quarry_CraneMediumBoom_A_Set_A) "RopeAttachSocket"))
  (LocalVariables|Setl_ropeDefaultLength (* _returnvalue_boxextent_x 2.0))
  (Transformation|SetWorldLocationAndRotation (Variables|Default|GetBoomRope) (LocalVariables|Getl_start) "-90.000000,0.000000,0.000000" false true)
  (LocalVariables|Setl_ropeScale (/ (Math|Vector|VectorLength (- (LocalVariables|Getl_end) (LocalVariables|Getl_start))) (LocalVariables|Getl_ropeDefaultLength)))
  (Transformation|SetWorldScale3D (Variables|Default|GetBoomRope) (Math|Vector|MakeVectorNetQuantize (LocalVariables|Getl_ropeScale) (.y (Transformation|GetWorldScale (Variables|Default|GetBoomRope))) (.z (Transformation|GetWorldScale (Variables|Default|GetBoomRope)))))
  (Rendering|Material|SetCustomPrimitiveDataFloat (Variables|Default|GetBoomRope) 0 (LocalVariables|Getl_ropeScale)))

(fn ValidateClawOpen
  (bind _clawopenamount (Variables|Default|GetClawOpenAmount))
  (bind _returnvalue (Math|Rotator|MakeRotator _clawopenamount))
  (Transformation|SetRelativeRotation (Variables|Default|GetSM_Quarry_CraneMediumClaw_A) _returnvalue false false)
  (Transformation|SetRelativeRotation (Variables|Default|GetSM_Quarry_CraneMediumClaw_B) (Math|Rotator|InvertRotator _returnvalue) false false))

(fn ValidateClawTransform
  (Transformation|SetRelativeLocation (Variables|Default|GetSM_Quarry_CraneMediumBoom_A_Set_A) (Math|Vector|MakeVector 0.0 0.0 (- (.z (.location (Transformation|GetSocketTransform (Variables|Default|GetSM_Quarry_CraneMediumBoom_A_Set_A) "RopeAttachSocket" "RTS_Component"))) (Math|Float|Absolute(Float) (Variables|Default|GetClawRopeLength)))) false true))
